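# Sound cue for the destruction of a Jolly Rascal weapon large prop in Fortnite, playing that prop's impact sound.
# A single wave player feeds the output directly, with a slight volume boost and no randomization.
# It uses Jolly Rascal attenuation and a third-person ranged weapon fire sound class.
# A concurrency override caps instances at three, taking precedence over the assigned one-per-owner, stop-oldest concurrency set.
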SoundCue /FNGameplayCues/Athena/Sounds/Weapons/Jolly_Rascal/Jolly_Rascal_Large_Prop_01_Destroy_Cue
root=WavePlayer_9
nodes [{id=WavePlayer_9, properties {soundWaveAssetPtr=&/WPN_Jolly_Rascal_Large_Prop_01_Impact}}]
properties {
 attenuationSettings=/FNGameplayCues/Sounds/Attenuation/Fort_Jolly_Rascal_Att
 bOverrideConcurrency=true
 maxDistance=107500
 soundClassObject=&Fort_WeaponRanged_Fire_3P_PSM
 volumeMultiplier=1.1
 concurrencyOverrides {maxCount=3}
 concurrencySet [/Game/Sounds/Concurrency_Rules/OneMaximum_Owner_StopOldest]
}
