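# Sound cue for the spawn of a large prop from the Jolly Rascal item, routed to the gadgets and items sound class.
# A short delay node feeds a single wave player, with the cue's volume boosted.
# Spatialization comes from the shared Jolly Rascal attenuation asset, as the cue's stored attenuation overrides are not switched on.
# A concurrency override caps simultaneous instances at a few.
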SoundCue /FNGameplayCues/Athena/Sounds/Weapons/Jolly_Rascal/Jolly_Rascal_Large_Prop_01_Spawn_Cue
root=Delay_0
nodes [
 {id=Delay_0, children [WavePlayer_0], properties {delayMin=0.1}}
 {id=WavePlayer_0, properties {soundWaveAssetPtr=&/WPN_Jolly_Rascal_Large_Prop_01_Spawn}}
]
properties {
 attenuationSettings=/FNGameplayCues/Sounds/Attenuation/Fort_Jolly_Rascal_Att
 bOverrideConcurrency=true
 maxDistance=107500
 soundClassObject=&Fort_Gadgets_Items
 volumeMultiplier=2
 attenuationOverrides {distanceAlgorithm=NaturalSound, falloffDistance=3800, lPFRadiusMax=3000, lPFRadiusMin=2000, attenuationShapeExtents {x=2200}}
 concurrencyOverrides {maxCount=3}
}
