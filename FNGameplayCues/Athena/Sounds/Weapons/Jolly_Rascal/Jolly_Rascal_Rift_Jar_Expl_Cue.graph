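# Sound cue for the Jolly Rascal Rift Jar explosion, routed to the gadgets and items sound class.
# A Mixer layers a smash wave with a rift-opening wave, which passes through a short fixed Delay so it lands just after the smash.
# Spatial falloff comes from the shared Jolly Rascal attenuation asset.
# A concurrency override caps simultaneous instances, stopping the oldest first.
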SoundCue /FNGameplayCues/Athena/Sounds/Weapons/Jolly_Rascal/Jolly_Rascal_Rift_Jar_Expl_Cue
root=Mixer_0
nodes [
 {id=Delay_0, children [WavePlayer_1], properties {delayMax=0.1, delayMin=0.1}}
 {id=Mixer_0, children [WavePlayer_0,Delay_0], properties {inputVolume [1,1]}}
 {id=WavePlayer_0, properties {soundWaveAssetPtr=&/WPN_Jolly_Rascal_Smash_01}}
 {id=WavePlayer_1, properties {soundWaveAssetPtr=&/WPN_Jolly_Rascal_Rift_Open_01}}
]
properties {
 attenuationSettings=/FNGameplayCues/Sounds/Attenuation/Fort_Jolly_Rascal_Att
 bOverrideConcurrency=true
 maxDistance=107500
 soundClassObject=&Fort_Gadgets_Items
 volumeMultiplier=1
 attenuationOverrides {distanceAlgorithm=NaturalSound, falloffDistance=3800, lPFRadiusMax=3000, lPFRadiusMin=2000, attenuationShapeExtents {x=2200}}
 concurrencyOverrides {maxCount=4, resolutionRule=StopOldest, volumeScale=0.5}
}
