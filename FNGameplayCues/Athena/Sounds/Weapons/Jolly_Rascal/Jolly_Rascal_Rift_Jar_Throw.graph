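# Sound cue for the throw of the Jolly Rascal Rift Jar gadget in Fortnite's Athena gameplay cues.
# A single wave player plays the throw sample directly, with no randomization or modulation nodes.
# It is routed to the gadgets and items sound class with gadget-placement attenuation, boosted volume, and its own concurrency override capping simultaneous instances.
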SoundCue /FNGameplayCues/Athena/Sounds/Weapons/Jolly_Rascal/Jolly_Rascal_Rift_Jar_Throw
root=WavePlayer_0
nodes [{id=WavePlayer_0, properties {soundWaveAssetPtr=&/WPN_Jolly_Rascal_Throw_01}}]
properties {
 attenuationSettings=/Game/Sounds/Attenuation/Fort_GadgetPlace_Att
 bOverrideConcurrency=true
 maxDistance=2000
 soundClassObject=&Fort_Gadgets_Items
 volumeMultiplier=2
 concurrencyOverrides {maxCount=4}
}
